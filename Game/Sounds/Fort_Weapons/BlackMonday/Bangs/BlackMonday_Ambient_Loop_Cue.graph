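# Sound cue for the Black Monday weapon's ambient loop, routed to the explosives sound class at reduced volume.
# A Random node picks with equal weight between two looping wave players that reuse Ashton Chicago throw loop waves from LTM audio.
# Inline overrides set logarithmic, short-range attenuation in place of the referenced fuse ambient attenuation asset, and cap concurrent instances at eight.
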SoundCue /Game/Sounds/Fort_Weapons/BlackMonday/Bangs/BlackMonday_Ambient_Loop_Cue
root=Random_0
nodes [
 {id=Random_0, children [WavePlayer_1,WavePlayer_2], properties {weights [1,1]}}
 {id=WavePlayer_1, properties {bLooping=true, soundWaveAssetPtr=/Game/Athena/Sounds/LTMs/Ashton/Chicago/Ashton_Chicago_Throw_Loop_01}}
 {id=WavePlayer_2, properties {bLooping=true, soundWaveAssetPtr=/Game/Athena/Sounds/LTMs/Ashton/Chicago/Ashton_Chicago_Throw_Loop_02}}
]
properties {
 attenuationSettings=/Game/Sounds/Attenuation/Fort_Explosive_Fuse_Ambient_Att
 bOverrideAttenuation=true
 bOverrideConcurrency=true
 maxDistance=1174
 soundClassObject=/Game/Sounds/FortSoundClassesAndModes/Fort_Explosives
 volumeMultiplier=0.2
 attenuationOverrides {distanceAlgorithm=Logarithmic, falloffDistance=1024, stereoSpread=1, attenuationShapeExtents {x=150}}
 concurrencyOverrides {maxCount=8}
}
