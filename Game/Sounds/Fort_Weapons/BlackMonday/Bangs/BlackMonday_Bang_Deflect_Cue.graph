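# Sound cue for the BlackMonday weapon's deflect bang, classed as an impact sound.
# A mixer layers two random picks, one of three main deflect waves and one of three layer waves, so repeated plays vary.
# The cue overrides attenuation with natural-sound falloff and caps concurrent instances through its own concurrency override.
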SoundCue /Game/Sounds/Fort_Weapons/BlackMonday/Bangs/BlackMonday_Bang_Deflect_Cue
root=Mixer_0
nodes [
 {id=Mixer_0, children [Random_1,Random_2], properties {inputVolume [1,1]}}
 {id=Random_1, children [WavePlayer_2,WavePlayer_4,WavePlayer_5], properties {weights [1,1,1]}}
 {id=Random_2, children [WavePlayer_6,WavePlayer_7,WavePlayer_8], properties {weights [1,1,1]}}
 {id=WavePlayer_2, properties {soundWaveAssetPtr=&/BlackMonday_Bang_Deflect_01}}
 {id=WavePlayer_4, properties {soundWaveAssetPtr=&/BlackMonday_Bang_Deflect_02}}
 {id=WavePlayer_5, properties {soundWaveAssetPtr=&/BlackMonday_Bang_Deflect_03}}
 {id=WavePlayer_6, properties {soundWaveAssetPtr=&/BlackMonday_Bang_Deflect_Layer_01}}
 {id=WavePlayer_7, properties {soundWaveAssetPtr=&/BlackMonday_Bang_Deflect_Layer_02}}
 {id=WavePlayer_8, properties {soundWaveAssetPtr=&/BlackMonday_Bang_Deflect_Layer_03}}
]
properties {
 attenuationSettings=&Fort_Weapon_Projectile_Medium_Mono_Athena_Att
 bOverrideAttenuation=true
 bOverrideConcurrency=true
 maxDistance=4800
 soundClassObject=&Fort_Impacts
 volumeMultiplier=0.6
 attenuationOverrides {distanceAlgorithm=NaturalSound, falloffDistance=4000, attenuationShapeExtents {x=800}}
 concurrencyOverrides {maxCount=4}
}
