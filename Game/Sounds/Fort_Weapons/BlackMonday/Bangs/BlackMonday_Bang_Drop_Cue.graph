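# Sound cue for the drop sound of the BlackMonday Bang weapon in Fortnite.
# A Random node picks one of three equally weighted drop wave variations, which feeds a Modulator that slightly varies volume.
# The cue uses shared weapon drop attenuation, the loot collect sound class and a reduced volume multiplier.
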SoundCue /Game/Sounds/Fort_Weapons/BlackMonday/Bangs/BlackMonday_Bang_Drop_Cue
root=Modulator_0
nodes [
 {id=Modulator_0, children [Random_1], properties {volumeMax=0.8, volumeMin=0.7}}
 {id=Random_1, children [WavePlayer_0,WavePlayer_3,WavePlayer_4], properties {weights [1,1,1]}}
 {id=WavePlayer_0, properties {soundWaveAssetPtr=&/BlackMonday_Bang_Drop_01}}
 {id=WavePlayer_3, properties {soundWaveAssetPtr=&/BlackMonday_Bang_Drop_02}}
 {id=WavePlayer_4, properties {soundWaveAssetPtr=&/BlackMonday_Bang_Drop_03}}
]
properties {
 attenuationSettings=&Fort_Weapon_Drop_Att
 maxDistance=3000
 soundClassObject=&Fort_Loot_Collect
 volumeMultiplier=0.4
}
