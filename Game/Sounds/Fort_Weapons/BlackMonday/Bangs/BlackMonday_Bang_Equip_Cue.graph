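# Sound cue for the equip sound of the BlackMonday "Bang" weapon, classed as player foley with close-prop attenuation.
# A mixer layers a fixed equip wave with a slightly delayed power-up wave, picked at random from two variants.
# A modulator randomizes the volume of the mix within a narrow range.
# A concurrency override prevents new instances from starting once the limit is reached.
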SoundCue /Game/Sounds/Fort_Weapons/BlackMonday/Bangs/BlackMonday_Bang_Equip_Cue
root=Modulator_1
nodes [
 {id=Delay_0, children [Random_1], properties {delayMax=0.1, delayMin=0.1}}
 {id=Mixer_0, children [WavePlayer_2,Delay_0], properties {inputVolume [1.1,0.8]}}
 {id=Modulator_1, children [Mixer_0], properties {volumeMax=0.85, volumeMin=0.75}}
 {id=Random_1, children [WavePlayer_4,WavePlayer_5], properties {weights [1,1]}}
 {id=WavePlayer_2, properties {soundWaveAssetPtr=&/BlackMonday_Bang_EquipB_01}}
 {id=WavePlayer_4, properties {soundWaveAssetPtr=&/BlackMonday_Bang_Equip_Powerup_01}}
 {id=WavePlayer_5, properties {soundWaveAssetPtr=&/BlackMonday_Bang_Equip_Powerup_02}}
]
properties {
 attenuationSettings=/Game/Sounds/Attenuation/Fort_Prop_Close_Att
 bOverrideConcurrency=true
 maxDistance=1800
 soundClassObject=&Fort_Player_Foley
 volumeMultiplier=1.3
 concurrencyOverrides {resolutionRule=PreventNew}
}
