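# Sound cue for the BlackMonday weapon's bang explosion, routed to the large explosives sound class.
# Each branch mixes close alternate explosions with quieter medium-attenuated distant ones, using random wave selection and volume modulation, plus pitch modulation on the distant layer.
# A quality node gives the first level the larger wave pool, while the other three share a two-plus-two variant branch.
# Concurrency allows one instance per owner, stopping the oldest.
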SoundCue /Game/Sounds/Fort_Weapons/BlackMonday/Bangs/BlackMonday_Bang_Explode_Cue
root=QualityLevel_0
nodes [
 {
  id=Attenuation_2
  children [Modulator_13]
  properties {attenuationSettings=&Fort_Weapon_Projectile_Medium_Mono_Athena_Att}
 }
 {id=Attenuation_3, children [Modulator_12], properties {attenuationSettings=&Fort_Weapon_Projectile_Close_Athena_Att}}
 {
  id=Attenuation_4
  children [Modulator_7]
  properties {attenuationSettings=&Fort_Weapon_Projectile_Medium_Mono_Athena_Att}
 }
 {id=Attenuation_5, children [Modulator_5], properties {attenuationSettings=&Fort_Weapon_Projectile_Close_Athena_Att}}
 {id=Mixer_2, children [Attenuation_3, Attenuation_2], properties {inputVolume [1,0.7]}}
 {id=Mixer_4, children [Attenuation_5, Attenuation_4], properties {inputVolume [1,0.7]}}
 {id=Modulator_12, children [Random_4], properties {volumeMax=1, volumeMin=0.9}}
 {id=Modulator_13, children [Random_8], properties {pitchMin=0.8, volumeMax=1, volumeMin=0.9}}
 {id=Modulator_5, children [Random_9], properties {volumeMax=1, volumeMin=0.9}}
 {id=Modulator_7, children [Random_10], properties {pitchMin=0.8, volumeMax=1, volumeMin=0.9}}
 {id=QualityLevel_0, children [Mixer_2,Mixer_4,Mixer_4,Mixer_4]}
 {id=Random_10, children [WavePlayer_36, WavePlayer_37], properties {weights [1,1]}}
 {id=Random_4, children [WavePlayer_12, WavePlayer_26, WavePlayer_27, WavePlayer_28], properties {weights [1,1,1,1]}}
 {id=Random_8, children [WavePlayer_29, WavePlayer_30, WavePlayer_31], properties {weights [1,1,1]}}
 {id=Random_9, children [WavePlayer_32, WavePlayer_33], properties {weights [1,1]}}
 {id=WavePlayer_12, properties {soundWaveAssetPtr=&/BlackMonday_Bang_Explode_Alt_02}}
 {id=WavePlayer_26, properties {soundWaveAssetPtr=&/BlackMonday_Bang_Explode_Alt_01}}
 {id=WavePlayer_27, properties {soundWaveAssetPtr=&/BlackMonday_Bang_Explode_Alt_03}}
 {id=WavePlayer_28, properties {soundWaveAssetPtr=&/BlackMonday_Bang_Explode_Alt_04}}
 {id=WavePlayer_29, properties {soundWaveAssetPtr=&/BlackMonday_Bang_Explode_Distant_01}}
 {id=WavePlayer_30, properties {soundWaveAssetPtr=&/BlackMonday_Bang_Explode_Distant_02}}
 {id=WavePlayer_31, properties {soundWaveAssetPtr=&/BlackMonday_Bang_Explode_Distant_03}}
 {id=WavePlayer_32, properties {soundWaveAssetPtr=&BlackMonday_Bang_Explode_Alt_02}}
 {id=WavePlayer_33, properties {soundWaveAssetPtr=&BlackMonday_Bang_Explode_Alt_01}}
 {id=WavePlayer_36, properties {soundWaveAssetPtr=&BlackMonday_Bang_Explode_Distant_01}}
 {id=WavePlayer_37, properties {soundWaveAssetPtr=&BlackMonday_Bang_Explode_Distant_02}}
]
properties {
 bOverrideConcurrency=true
 maxDistance=10579.2871
 soundClassObject=/Game/Sounds/FortSoundClassesAndModes/Fort_Explosives_Large_PSM
 volumeMultiplier=0.7076
 concurrencyOverrides {bLimitToOwner=true, maxCount=1, resolutionRule=StopOldest}
}
