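# BlackMonday bang body-impact SoundCue that plays one of four impact wave variations.
# A Random node with equal weights picks among the four wave players on each trigger.
# The cue overrides attenuation with distance falloff and low-pass filtering, and caps concurrency at four instances, stopping the oldest.
# It routes through the third-person ranged-weapon fire sound class.
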SoundCue /Game/Sounds/Fort_Weapons/BlackMonday/Bangs/BlackMonday_Bang_ImpactBody_Cue
root=Random_2
nodes [
 {id=Random_2, children [WavePlayer_8, WavePlayer_9, WavePlayer_10, WavePlayer_11], properties {weights [1,1,1,1]}}
 {id=WavePlayer_10, properties {soundWaveAssetPtr=&/BlackMonday_Bang_ImpactBody_03}}
 {id=WavePlayer_11, properties {soundWaveAssetPtr=&/BlackMonday_Bang_ImpactBody_04}}
 {id=WavePlayer_8, properties {soundWaveAssetPtr=&/BlackMonday_Bang_ImpactBody_01}}
 {id=WavePlayer_9, properties {soundWaveAssetPtr=&/BlackMonday_Bang_ImpactBody_02}}
]
properties {
 bOverrideAttenuation=true
 bOverrideConcurrency=true
 maxDistance=4200
 soundClassObject=&Fort_WeaponRanged_Fire_3P_PSM
 volumeMultiplier=1
 attenuationOverrides {falloffDistance=3400, lPFRadiusMax=3600, lPFRadiusMin=1800, attenuationShapeExtents {x=800}}
 concurrencyOverrides {maxCount=4, resolutionRule=StopOldest}
}
