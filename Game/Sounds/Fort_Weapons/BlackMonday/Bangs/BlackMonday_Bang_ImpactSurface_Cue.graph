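# Sound cue for the BlackMonday Bang weapon's surface impact in the Fort_Weapons set, routed to the Fort_Impacts sound class.
# A mixer layers a random pick of four equally weighted impact variations with a quieter activation sound that follows after a short fixed delay.
# Overrides set distance falloff with low-pass filtering and cap how many instances play concurrently.
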SoundCue /Game/Sounds/Fort_Weapons/BlackMonday/Bangs/BlackMonday_Bang_ImpactSurface_Cue
root=Mixer_1
nodes [
 {id=Delay_0, children [WavePlayer_0], properties {delayMax=0.1, delayMin=0.1}}
 {id=Mixer_1, children [Random_1,Delay_0], properties {inputVolume [1.2,0.7]}}
 {id=Random_1, children [WavePlayer_2,WavePlayer_3,WavePlayer_4,WavePlayer_5], properties {weights [1,1,1,1]}}
 {id=WavePlayer_0, properties {soundWaveAssetPtr=&/BlackMonday_Bang_Activate_01}}
 {id=WavePlayer_2, properties {soundWaveAssetPtr=&/BlackMonday_Bang_ImpactSurface_01}}
 {id=WavePlayer_3, properties {soundWaveAssetPtr=&/BlackMonday_Bang_ImpactSurface_02}}
 {id=WavePlayer_4, properties {soundWaveAssetPtr=&/BlackMonday_Bang_ImpactSurface_03}}
 {id=WavePlayer_5, properties {soundWaveAssetPtr=&/BlackMonday_Bang_ImpactSurface_04}}
]
properties {
 attenuationSettings=&Fort_Weapon_Melee_Sml_Att
 bOverrideAttenuation=true
 bOverrideConcurrency=true
 maxDistance=4200
 soundClassObject=&Fort_Impacts
 volumeMultiplier=1.1
 attenuationOverrides {bAttenuateWithLPF=true, falloffDistance=3400, lPFRadiusMax=3600, lPFRadiusMin=1800, attenuationShapeExtents {x=800}}
 concurrencyOverrides {maxCount=4}
}
