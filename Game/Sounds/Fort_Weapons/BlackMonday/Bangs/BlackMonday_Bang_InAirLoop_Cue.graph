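# Sound cue for the looping in-air sound of the BlackMonday Bang weapon projectile, classed as a ranged projectile effect.
# A Mixer blends two looping wave layers, a main in-air loop and an extra layer, at slightly reduced volumes.
# The mix feeds a root Attenuation node using close projectile settings, while the cue itself references a Batarang-specific attenuation asset.
# The cue overrides concurrency to prevent new instances once the limit is reached.
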SoundCue /Game/Sounds/Fort_Weapons/BlackMonday/Bangs/BlackMonday_Bang_InAirLoop_Cue
root=Attenuation_0
nodes [
 {id=Attenuation_0, children [Mixer_0], properties {attenuationSettings=&Fort_Weapon_Projectile_Close_Athena_Att}}
 {id=Mixer_0, children [WavePlayer_2,WavePlayer_0], properties {inputVolume [0.9,0.8]}}
 {
  id=WavePlayer_0
  properties {bLooping=true, soundWaveAssetPtr=&/BlackMonday_Bang_InAirLoopLayer_01}
 }
 {id=WavePlayer_2, properties {bLooping=true, soundWaveAssetPtr=&/BlackMonday_Bang_InAirLoop_01}}
]
properties {
 attenuationSettings=/Game/Sounds/Attenuation/Fort_Weapon_Projectile_Close_Athena_Batarang_Att
 bOverrideConcurrency=true
 maxDistance=4950
 soundClassObject=&Fort_WeaponRanged_Projectile_Effects
 volumeMultiplier=0.85
 attenuationOverrides {falloffDistance=5000, attenuationShapeExtents {x=1000}}
 concurrencyOverrides {resolutionRule=PreventNew}
}
