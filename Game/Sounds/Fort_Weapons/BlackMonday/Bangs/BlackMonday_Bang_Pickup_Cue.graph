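# Sound cue for the Black Monday Bangs weapon pickup sound, judging by its name and loot-collect sound class.
# A Random node chooses among three pickup wave variations with equal weight.
# It uses the shared weapon pickup attenuation at reduced volume, and a concurrency override prevents new instances from starting once the limit is reached.
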SoundCue /Game/Sounds/Fort_Weapons/BlackMonday/Bangs/BlackMonday_Bang_Pickup_Cue
root=Random_0
nodes [
 {id=Random_0, children [WavePlayer_1,WavePlayer_2,WavePlayer_3], properties {weights [1,1,1]}}
 {id=WavePlayer_1, properties {soundWaveAssetPtr=&/BlackMonday_Bang_Pickup_01}}
 {id=WavePlayer_2, properties {soundWaveAssetPtr=&/BlackMonday_Bang_Pickup_02}}
 {id=WavePlayer_3, properties {soundWaveAssetPtr=&/BlackMonday_Bang_Pickup_03}}
]
properties {
 attenuationSettings=&Fort_Weapon_Pickup_Att
 bOverrideConcurrency=true
 maxDistance=3500
 soundClassObject=&Fort_Loot_Collect
 volumeMultiplier=0.3446
 concurrencyOverrides {resolutionRule=PreventNew}
}
