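# Sound cue for the BlackMonday weapon's throw bang, mixing a close layer with a distant layer.
# Each layer picks a random wave, varies it through a modulator, then applies its own projectile attenuation and ranged-fire sound class, with the distant layer quieter.
# A quality-level switch gives the first level four variations per layer, while the other three levels share a two-variation branch.
# The cue overrides concurrency to prevent new sounds at the limit.
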SoundCue /Game/Sounds/Fort_Weapons/BlackMonday/Bangs/BlackMonday_Bang_Throw_Cue
root=QualityLevel_0
nodes [
 {id=Attenuation_3, children [Modulator_6], properties {attenuationSettings=&Fort_Weapon_Projectile_Close_Athena_Att}}
 {
  id=Attenuation_4
  children [Modulator_7]
  properties {attenuationSettings=&Fort_Weapon_Projectile_Medium_Mono_Athena_Att}
 }
 {
  id=Attenuation_5
  children [Modulator_9]
  properties {attenuationSettings=&Fort_Weapon_Projectile_Medium_Mono_Athena_Att}
 }
 {id=Attenuation_6, children [Modulator_8], properties {attenuationSettings=&Fort_Weapon_Projectile_Close_Athena_Att}}
 {id=Mixer_1, children [SoundClass_3,SoundClass_2], properties {inputVolume [0.8,0.8]}}
 {id=Mixer_6, children [SoundClass_5,SoundClass_4], properties {inputVolume [0.8,0.8]}}
 {id=Modulator_6, children [Random_1], properties {volumeMax=1, volumeMin=0.9}}
 {id=Modulator_7, children [Random_2], properties {volumeMax=0.8, volumeMin=0.7}}
 {id=Modulator_8, children [Random_4], properties {volumeMax=1, volumeMin=0.9}}
 {id=Modulator_9, children [Random_6], properties {volumeMax=0.8, volumeMin=0.7}}
 {id=QualityLevel_0, children [Mixer_1,Mixer_6,Mixer_6,Mixer_6]}
 {id=Random_1, children [WavePlayer_2,WavePlayer_3,WavePlayer_4,WavePlayer_5], properties {weights [1,1,1,1]}}
 {id=Random_2, children [WavePlayer_6, WavePlayer_8, WavePlayer_9, WavePlayer_12], properties {weights [1,1,1,1]}}
 {id=Random_4, children [WavePlayer_15, WavePlayer_16], properties {weights [1,1]}}
 {id=Random_6, children [WavePlayer_29, WavePlayer_30], properties {weights [1,1]}}
 {
  id=SoundClass_2
  children [Attenuation_4]
  properties {soundClassOverride=&Fort_WeaponRanged_Fire_Distant}
 }
 {
  id=SoundClass_3
  children [Attenuation_3]
  properties {soundClassOverride=&Fort_WeaponRanged_Fire_3P_PSM}
 }
 {
  id=SoundClass_4
  children [Attenuation_5]
  properties {soundClassOverride=&Fort_WeaponRanged_Fire_Distant}
 }
 {
  id=SoundClass_5
  children [Attenuation_6]
  properties {soundClassOverride=&Fort_WeaponRanged_Fire_3P_PSM}
 }
 {id=WavePlayer_12, properties {soundWaveAssetPtr=&/BlackMonday_Bang_Throw_Distant_04}}
 {id=WavePlayer_15, properties {soundWaveAssetPtr=&/BlackMonday_Bang_Throw_01}}
 {id=WavePlayer_16, properties {soundWaveAssetPtr=&/BlackMonday_Bang_Throw_02}}
 {id=WavePlayer_2, properties {soundWaveAssetPtr=&BlackMonday_Bang_Throw_01}}
 {id=WavePlayer_29, properties {soundWaveAssetPtr=&/BlackMonday_Bang_Throw_Distant_01}}
 {id=WavePlayer_3, properties {soundWaveAssetPtr=&BlackMonday_Bang_Throw_02}}
 {id=WavePlayer_30, properties {soundWaveAssetPtr=&/BlackMonday_Bang_Throw_Distant_02}}
 {id=WavePlayer_4, properties {soundWaveAssetPtr=&/BlackMonday_Bang_Throw_03}}
 {id=WavePlayer_5, properties {soundWaveAssetPtr=&/BlackMonday_Bang_Throw_04}}
 {id=WavePlayer_6, properties {soundWaveAssetPtr=&BlackMonday_Bang_Throw_Distant_01}}
 {id=WavePlayer_8, properties {soundWaveAssetPtr=&BlackMonday_Bang_Throw_Distant_02}}
 {id=WavePlayer_9, properties {soundWaveAssetPtr=&/BlackMonday_Bang_Throw_Distant_03}}
]
properties {
 bOverrideConcurrency=true
 maxDistance=10579.2871
 attenuationOverrides {falloffDistance=5000, attenuationShapeExtents {x=1000}}
 concurrencyOverrides {resolutionRule=PreventNew}
}
